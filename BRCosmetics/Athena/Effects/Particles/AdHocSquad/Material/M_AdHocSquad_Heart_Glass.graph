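# Translucent, two-sided glass material for Niagara mesh particles, apparently the heart in the AdHocSquad cosmetic effect.
# A fake matcap-style reflection crosses view-space position and normal, using the G component for both UVs of a reflection texture.
# Emissive adds that reflection, tinted by offset particle color, to a particle-colored Fresnel glow, then runs through MF_FX_EmissiveAdjustments.
# Opacity lerps between Fresnel and a backface scalar using raw TwoSidedSign as alpha.
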Material /BRCosmetics/Athena/Effects/Particles/AdHocSquad/Material/M_AdHocSquad_Heart_Glass
graph {
 properties {
  bAutomaticallySetUsageInEditor=false
  bUsedWithNiagaraMeshParticles=true
  blendMode=BLEND_Translucent
  translucencyLightingMode=TLM_VolumetricPerVertexDirectional
  twoSided=true
  usedShadingModels=Default Lit
 }
 nodes [
  LinearInterpolate(
   A=Fresnel(
    ExponentIn=ScalarParameter() {defaultValue=1.4, group=Opacity, parameterName=Opacity Fresnel Exp}
    BaseReflectFractionIn=ScalarParameter() {defaultValue=0.15, group=Opacity, parameterName=Opacity Fresnel Base}
   )
   B=ScalarParameter() {defaultValue=0.6, group=Opacity, parameterName=Opacity Backface}
   Alpha=TwoSidedSign()
  )
  n1 = TransformPosition(None=WorldPosition().XYZ) {transformSourceType=TRANSFORMPOSSOURCE_World, transformType=TRANSFORMPOSSOURCE_View}
  n2 = VertexNormalWS()
  n3 = MaterialFunctionCall(Color=CrossProduct(A=Normalize(VectorInput=@n1), B=Transform(None=@n2) {transformSourceType=TRANSFORMSOURCE_World, transformType=TRANSFORM_View})) {materialFunction=&SplitComponents}
  n4 = AppendVector(A=@n3.G, B=@n3.G)
  n5 = TextureSample(UVs=Add(A=Multiply(A=@n4) {constB=0.5}) {constB=0.5}) {texture=/BRCosmetics/Athena/Effects/Particles/AdHocSquad/Texture/T_Reflection_AdHocSquad}
  n6 = ParticleColor()
  n7 = Add(A=@n6.RGB) {constB=0.1}
  n8 = Fresnel(ExponentIn=ScalarParameter() {defaultValue=10, group=Reflection, parameterName=Fresnel Glow Exp}) {baseReflectFraction=0}
  MaterialFunctionCall(Emissive=Add(A=Multiply(A=@n5.RGB, B=@n7), B=LinearInterpolate(B=@n6.RGB, Alpha=@n8))) {materialFunction=&MF_FX_EmissiveAdjustments}
 ]
}
