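# M_DiamondFlash is an unlit translucent Niagara sprite material that draws a soft diamond-shaped flash for AdHocSquad cosmetic effects.
# The shape is fully procedural: center-scaled UVs are mirrored into a Manhattan-distance diamond, sharpened by an Exponent parameter and multiplied by a powered radial SphereMask.
# Particle color tints the mask through MF_FX_EmissiveAdjustments, while particle alpha multiplies a depth-faded copy to soften intersections.
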Material /BRCosmetics/Athena/Effects/Particles/AdHocSquad/Material/M_DiamondFlash
graph {
 properties {bUsedWithNiagaraSprites=true, blendMode=BLEND_Translucent, pixelDepthOffsetMode=PDOM_Legacy, shadingModel=MSM_Unlit, usedShadingModels=Unlit}
 nodes [
  n1 = ParticleColor()
  n2 = MaterialFunctionCall(UVs=TextureCoordinate(), Texture Scale=Constant() {r=2.2}) {materialFunction=&ScaleUVsByCenter}
  n3 = Multiply(A=Abs(None=Subtract(A=@n2.UVs) {constB=0.5})) {constB=2}
  n4 = Subtract(A=OneMinus(None=ComponentMask(None=@n3) {r=true}), B=ComponentMask(None=@n3) {g=true})
  n5 = ScalarParameter() {defaultValue=3, parameterName=Exponent}
  n6 = Constant() {r=0.5}
  n7 = SphereMask(A=@n2.UVs, B=@n6, Radius=@n6) {hardnessPercent=0}
  n8 = Multiply(A=Multiply(A=Power(Base=@n4, Exp=@n5), B=Power(Base=@n7) {constExponent=3}))
  MaterialFunctionCall(Emissive=Multiply(A=@n1.RGB, B=@n8)) {materialFunction=&MF_FX_EmissiveAdjustments}
  Multiply(A=@n1.A, B=DepthFade(Opacity=@n8))
 ]
}
