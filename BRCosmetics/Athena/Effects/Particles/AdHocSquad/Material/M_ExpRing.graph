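# M_ExpRing is an additive, unlit material for Niagara sprite particles that draws a ring effect, part of the AdHocSquad cosmetic effects.
# The red channel of the T_VFX_ExpRing texture serves as the mask, tinted by particle color and passed through MF_FX_EmissiveAdjustments for emissive.
# The same mask goes through a depth fade and is multiplied by particle alpha, softening intersections with scene geometry.
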Material /BRCosmetics/Athena/Effects/Particles/AdHocSquad/Material/M_ExpRing
graph {
 properties {bUsedWithNiagaraSprites=true, blendMode=BLEND_Additive, pixelDepthOffsetMode=PDOM_Legacy, shadingModel=MSM_Unlit, usedShadingModels=Unlit}
 nodes [
  n1 = ParticleColor()
  n2 = TextureSample() {texture=/BRCosmetics/Athena/Effects/Particles/AdHocSquad/Texture/T_VFX_ExpRing}
  MaterialFunctionCall(Emissive=Multiply(A=@n1.RGB, B=@n2.R)) {materialFunction=&MF_FX_EmissiveAdjustments}
  Multiply(A=@n1.A, B=DepthFade(Opacity=@n2.R))
 ]
}
