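# Additive, unlit lens flare material for Niagara sprite particles in the AdHocSquad cosmetic effects.
# Samples the lens flare texture three times at increasing UV scales around the sprite center, packing the red channels into RGB to fake chromatic aberration.
# That color is tinted by particle color and passed through MF_FX_EmissiveAdjustments.
# The maximum of the three samples is multiplied by particle alpha to form a mask.
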Material /BRCosmetics/Athena/Effects/Particles/AdHocSquad/Material/M_Flare
graph {
 properties {bUsedWithNiagaraSprites=true, blendMode=BLEND_Additive, pixelDepthOffsetMode=PDOM_Legacy, shadingModel=MSM_Unlit, usedShadingModels=Unlit}
 nodes [
  n1 = Subtract(A=TextureCoordinate()) {constB=0.5}
  n2 = TextureObject() {texture=/BRCosmetics/Athena/Effects/Particles/AdHocSquad/Texture/T_VFX_LensFlare}
  n3 = TextureSample(UVs=Add(A=Multiply(A=@n1)) {constB=0.5}, Tex=@n2) {samplerSource=SSM_Clamp_WorldGroupSettings, texture=&T_VFX_LensFlare}
  n4 = TextureSample(UVs=Add(A=Multiply(A=@n1) {constB=1.2}) {constB=0.5}, Tex=@n2) {samplerSource=SSM_Clamp_WorldGroupSettings, texture=&T_VFX_LensFlare}
  n5 = AppendVector(A=@n3.R, B=@n4.R)
  n6 = TextureSample(UVs=Add(A=Multiply(A=@n1) {constB=1.4}) {constB=0.5}, Tex=@n2) {samplerSource=SSM_Clamp_WorldGroupSettings, texture=&T_VFX_LensFlare}
  n7 = ParticleColor()
  MaterialFunctionCall(Emissive=Multiply(A=AppendVector(A=@n5, B=@n6.R), B=@n7.RGB)) {materialFunction=&MF_FX_EmissiveAdjustments}
  Multiply(A=Max(A=Max(A=@n3.R, B=@n4.R), B=@n6.R), B=@n7.A)
 ]
}
