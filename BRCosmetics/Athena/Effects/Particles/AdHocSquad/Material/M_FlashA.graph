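# M_FlashA is an unlit translucent Niagara sprite material for a flare flash effect in the AdHocSquad cosmetic.
# Particle color times a soft, powered sphere mask feeds an emissive adjustment function.
# Two flare textures, randomly rotated per particle and one rescaled about center, are blended and masked by that sphere.
# A lens flare texture is lightly lerped in, and particle alpha scales the result.
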Material /BRCosmetics/Athena/Effects/Particles/AdHocSquad/Material/M_FlashA
graph {
 properties {bUsedWithNiagaraSprites=true, blendMode=BLEND_Translucent, pixelDepthOffsetMode=PDOM_Legacy, shadingModel=MSM_Unlit, usedShadingModels=Unlit}
 nodes [
  n1 = ParticleColor()
  n2 = Constant() {r=0.5}
  n3 = Power(Base=SphereMask(A=TextureCoordinate(), B=@n2, Radius=@n2) {hardnessPercent=0}) {constExponent=3}
  MaterialFunctionCall(Emissive=Multiply(A=@n1.RGB, B=@n3)) {materialFunction=&MF_FX_EmissiveAdjustments}
  n4 = MaterialFunctionCall(UVs=TextureCoordinate(), Rotation Center=Constant() {r=0.5}, "Rotation Angle (0-1)"=ParticleRandom()) {materialFunction=&CustomRotator}
  n5 = Constant() {r=0.7}
  n6 = LinearInterpolate(
   A=TextureSample(UVs=MaterialFunctionCall(UVs=@n4.Rotated Values, Texture Scale=@n5) {materialFunction=&ScaleUVsByCenter}.UVs) {samplerType=SAMPLERTYPE_LinearGrayscale, texture=/BRCosmetics/Athena/Effects/Particles/AdHocSquad/Texture/T_VFX_Flare_A}.R
   B=TextureSample(UVs=@n4.Rotated Values) {samplerType=SAMPLERTYPE_Grayscale, texture=/BRCosmetics/Athena/Effects/Particles/AdHocSquad/Texture/T_VFX_Flare_B}.R
  )
  Multiply(
   A=@n1.A
   B=LinearInterpolate(A=Multiply(A=@n6, B=@n3), B=TextureSample() {texture=/BRCosmetics/Athena/Effects/Particles/AdHocSquad/Texture/T_VFX_LensFlare}.R) {constAlpha=0.25}
  )
 ]
}
