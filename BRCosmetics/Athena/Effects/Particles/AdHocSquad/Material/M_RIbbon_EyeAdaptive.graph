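# Unlit additive material for Niagara ribbons and sprites in the AdHocSquad cosmetic particle effects.
# Particle color RGB feeds the MF_FX_EmissiveAdjustments function for glow, which the asset name suggests compensates for eye adaptation.
# Particle alpha is multiplied by inverted U texture coordinate, producing a linear fade along that axis without textures.
# Responsive AA and translucent velocity output are enabled.
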Material /BRCosmetics/Athena/Effects/Particles/AdHocSquad/Material/M_RIbbon_EyeAdaptive
graph {
 properties {
  bEnableResponsiveAA=true
  bOutputTranslucentVelocity=true
  bUsedWithNiagaraRibbons=true
  bUsedWithNiagaraSprites=true
  blendMode=BLEND_Additive
  pixelDepthOffsetMode=PDOM_Legacy
  shadingModel=MSM_Unlit
  usedShadingModels=Unlit
 }
 nodes [
  n1 = ParticleColor()
  MaterialFunctionCall(Emissive=@n1.RGB) {materialFunction=&MF_FX_EmissiveAdjustments}
  n2 = TextureCoordinate()
  Multiply(A=@n1.A, B=OneMinus(None=ComponentMask(None=@n2) {r=true}))
 ]
}
