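# Unlit translucent material for Niagara mesh particles, drawing a glowing spawn ring in the AdHocSquad effects.
# It shapes a soft band from mesh V, using floor to pick different falloff exponents either side of the peak.
# A T_Wind sample along U at fixed V feeds the named Texture reroute as streak variation.
# Particle color and alpha scale the smoothstepped band, with emissive passing through MF_FX_EmissiveAdjustments.
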Material /BRCosmetics/Athena/Effects/Particles/AdHocSquad/Material/M_SpawnRing
graph {
 properties {
  bUsedWithNiagaraMeshParticles=true
  blendMode=BLEND_Translucent
  pixelDepthOffsetMode=PDOM_Legacy
  shadingModel=MSM_Unlit
  usedShadingModels=Unlit
 }
 nodes [
  n1 = TextureCoordinate()
  ComponentMask(None=@n1) {g=true}
  n2 = NamedRerouteUsage()
  n3 = ParticleColor()
  n4 = OneMinus(None=ComponentMask(None=TextureCoordinate()) {g=true})
  n5 = Multiply(A=OneMinus(None=Power(Base=@n4))) {constB=2}
  n6 = Abs(None=Subtract(A=@n5))
  n7 = NamedRerouteUsage()
  n8 = LinearInterpolate(Alpha=@n7) {constA=10}
  n9 = LinearInterpolate(Alpha=@n7) {constA=3}
  n10 = Floor(None=@n5)
  n11 = SmoothStep(Value=Power(Base=OneMinus(None=@n6), Exp=LinearInterpolate(A=@n8, B=@n9, Alpha=@n10)) {constExponent=3})
  MaterialFunctionCall(Emissive=Multiply(A=Multiply(A=@n2, B=@n3.RGB), B=@n11)) {materialFunction=&MF_FX_EmissiveAdjustments}
  Multiply(A=@n3.A, B=@n11)
  n12 = Multiply(A=ComponentMask(None=@n1) {r=true}) {constB=2}
  n13 = Constant() {r=0.424}
  NamedRerouteDeclaration(None=TextureSample(UVs=AppendVector(A=@n12, B=@n13)) {texture=/BRCosmetics/Athena/Effects/Particles/AdHocSquad/Texture/T_Wind}.R) {name=Texture, nodeColor=<rgba 0 1 0.2734 1>}
 ]
}
