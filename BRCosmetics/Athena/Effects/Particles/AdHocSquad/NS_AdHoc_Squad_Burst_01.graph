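# Niagara cosmetic effect for AdHoc Squad: a spinning heart that charges, rises and bursts into purple-pink sparks.
# Swirling orbs with ribbon trails spiral inward around a glow and growing inner heart until a flash and ring fire.
# The heart then pulses upward with halo and glitter, ending in flares, sparks and twinkles.
# Emitters synchronize through fixed burst times and age-indexed curves, sharing System.Offset, with System.EndHeight setting the rise and burst height.
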NiagaraSystem /BRCosmetics/Athena/Effects/Particles/AdHocSquad/NS_AdHoc_Squad_Burst_01
dependencies {
 usedDataInterfaces [&NiagaraDataInterfaceCurve, &NiagaraDataInterfaceColorCurve]
 usedDynamicInputs []
 usedModules [
  &self:SystemSpawnScript.SystemScriptSource.SystemScriptGraph.NiagaraNodeAssignment_0.SetVariables_#1
  &self:SystemSpawnScript.SystemScriptSource.SystemScriptGraph.NiagaraNodeAssignment_1.SetVariables_#2
  &SystemState
 ]
}
emitters {
 Burst_Sparks {
  simTarget=GPUComputeSim
  emitterUpdateScript [
   &EmitterState()
   &SpawnBurst_Instantaneous(Spawn Count=200, Spawn Time=2.4, Use Spawn Probability=false, UseLoopCountLimit=false)
  ]
  particleSpawnScript [
   &InitializeParticle(
    Color=<rgba 35.9195 2.6087 100 1>
    Lifetime Mode=::NewEnumerator1
    Position Offset=&AddVectorToPosition(Position=$System.Offset, Vector=&MakeVector(X=0, Y=0, Z=$System.EndHeight))
    Position Offset Coordinate Space=::Local
    Sprite Rotation Mode=::NewEnumerator2
    Sprite Size Mode=::NewEnumerator1
    Uniform Sprite Size Max=3
    Uniform Sprite Size Min=1.5
    UsePositionOffset=true
    Write Material Random=true
   )
   &ShapeLocation(MaskSpawnGroup=false, Override Seed=false, Sphere Radius=5, UseOwnerScale=true)
   &AddVelocity(
    Module.Use Falloff Distance=false
    Origin Offset=&AddVectorToPosition(
     Position=$System.Offset
     Vector=&MakeVector(X=0, Y=0, Z=&Add_Float(A=-1, B=$System.EndHeight))
    )
    Override Seed=false
    Velocity Mode=::NewEnumerator1
    Velocity Speed=&RandomRangeFloat(Minimum=200, Maximum=450)
   )
   &ApplyInitialForces()
  ]
  particleUpdateScript [
   &ParticleState()
   &ScaleColor(
    Scale Alpha=&FloatFromCurve(FloatCurve=DataInterface() {propertyValues {Curve {defaultValue=1, keys [{tangentMode=RCTM_User, time=0, value=0}, {tangentMode=RCTM_User, time=0.5, value=1}]}}})
    ScaleA=true
    ScaleRGB=true
   )
   &ScaleSpriteSize(
    Uniform Curve Scale=&Sine(
     Normalized Angle=$Particles.NormalizedAge
     Period=&RandomRangeFloat(Minimum=0.4, Maximum=0.6)
     Scale=0.5
     Bias=1
    )
    Uniform Curve Sprite Scale=DataInterface() {propertyValues {Curve {defaultValue=0, keys [{tangentMode=RCTM_User, time=0.25, value=1}, {tangentMode=RCTM_User, time=1, value=0}]}}}
   )
   &Drag(Drag=&RandomRangeFloat(Minimum=3, Maximum=6), Use Linear Drag=true, Use Rotational Drag=true)
   &GravityForce(Gravity=&MakeVector(X=0, Y=0, Z=&RandomRangeFloat(Minimum=-245, Maximum=-490)))
   &SolveForcesAndVelocity(Clamp Velocity=false, Limit Acceleration=false)
  ]
  rendererClasses [&NiagaraSpriteRendererProperties]
  renderers [{rendererClass=&NiagaraSpriteRendererProperties}]
 }
 Burst_Sparks_B {
  simTarget=GPUComputeSim
  emitterUpdateScript [
   &EmitterState()
   &SpawnBurst_Instantaneous(Spawn Count=200, Spawn Time=2.3, Use Spawn Probability=false, UseLoopCountLimit=false)
  ]
  particleSpawnScript [
   &InitializeParticle(
    Color=<rgba 17.1015 0 50 0.75>
    Lifetime Max=0.7
    Lifetime Min=0.3
    Lifetime Mode=::NewEnumerator1
    Position Offset=&AddVectorToPosition(Position=$System.Offset, Vector=&MakeVector(X=0, Y=0, Z=$System.EndHeight))
    Position Offset Coordinate Space=::Local
    Sprite Rotation Mode=::NewEnumerator2
    Sprite Size Mode=::NewEnumerator1
    Uniform Sprite Size Max=2
    Uniform Sprite Size Min=1
    UsePositionOffset=true
    Write Material Random=true
   )
   &ShapeLocation(MaskSpawnGroup=false, Override Seed=false, Sphere Radius=2.5, UseOwnerScale=true)
   &AddVelocity(
    Module.Use Falloff Distance=false
    Origin Offset=&AddVectorToPosition(
     Position=$System.Offset
     Vector=&MakeVector(X=0, Y=0, Z=&Add_Float(A=0, B=$System.EndHeight))
    )
    Override Seed=false
    Velocity Mode=::NewEnumerator1
    Velocity Speed=&RandomRangeFloat(Minimum=100, Maximum=350)
   )
   &ApplyInitialForces()
  ]
  particleUpdateScript [
   &Drag(Drag=2.3, Use Linear Drag=true, Use Rotational Drag=false)
   &GravityForce(Gravity=<0 0 -500>) {enabled=false}
   &SolveForcesAndVelocity(Clamp Velocity=false, Limit Acceleration=false)
   &ParticleState()
   &ScaleSpriteSizeBySpeed(Max Scale Factor=<0.65 13>, Min Scale Factor=<0.68 0.72>)
   &ScaleColor(
    Scale Alpha=&FloatFromCurve(FloatCurve=DataInterface() {propertyValues {Curve {keys [<at 0 1 RCIM_Cubic>,<at 1 0>]}}})
    ScaleA=true
    ScaleRGB=true
   )
  ]
  rendererClasses [&NiagaraSpriteRendererProperties]
  renderers [{rendererClass=&NiagaraSpriteRendererProperties}]
 }
 Burst_Twinkles {
  simTarget=GPUComputeSim
  emitterUpdateScript [
   &EmitterState()
   &SpawnBurst_Instantaneous(Spawn Count=400, Spawn Time=2.4, Use Spawn Probability=false, UseLoopCountLimit=false)
  ]
  particleSpawnScript [
   &InitializeParticle(
    Color=<rgba 17.1015 0 50 0.75>
    Lifetime Max=2.5
    Lifetime Min=1.25
    Lifetime Mode=::NewEnumerator1
    Position Offset=&AddVectorToPosition(Position=$System.Offset, Vector=&MakeVector(X=0, Y=0, Z=$System.EndHeight))
    Position Offset Coordinate Space=::Local
    Sprite Rotation Mode=::NewEnumerator2
    Sprite Size Mode=::NewEnumerator1
    Uniform Sprite Size Max=2
    Uniform Sprite Size Min=1
    UsePositionOffset=true
    Write Material Random=true
   )
   &ShapeLocation(MaskSpawnGroup=false, Override Seed=false, Sphere Radius=5, UseOwnerScale=true)
   &AddVelocity(
    Module.Use Falloff Distance=false
    Origin Offset=&AddVectorToPosition(
     Position=$System.Offset
     Vector=&MakeVector(X=0, Y=0, Z=&Add_Float(A=0, B=$System.EndHeight))
    )
    Override Seed=false
    Velocity Mode=::NewEnumerator1
    Velocity Speed=&RandomRangeFloat(Minimum=150, Maximum=550)
   )
   &ApplyInitialForces()
  ]
  particleUpdateScript [
   &ParticleState()
   &ScaleColor(ScaleA=true, ScaleRGB=true)
   &ScaleSpriteSize(
    Uniform Curve Scale=&FloatFromCurve(
     FloatCurve=DataInterface() {propertyValues {Curve {keys [<at 0 1 RCIM_Cubic>,<at 0.5 0 RCIM_Cubic>,<at 1 1 RCIM_Cubic>]}}}
     CurveIndex=&Frac_Float(Float=&Multiply_Float(A=$Emitter.Age, B=&RandomRangeFloat(Minimum=2, Maximum=4)))
    )
    Uniform Curve Sprite Scale=DataInterface() {propertyValues {Curve {defaultValue=0, keys [{tangentMode=RCTM_User, time=0.75, value=1}, {tangentMode=RCTM_User, time=1, value=0}]}}}
   )
   &Drag(Drag=&RandomRangeFloat(Minimum=5, Maximum=7), Use Linear Drag=true, Use Rotational Drag=true)
   &GravityForce(Gravity=&MakeVector(X=0, Y=0, Z=&RandomRangeFloat(Minimum=-15, Maximum=-30)))
   &SolveForcesAndVelocity(Clamp Velocity=false, Limit Acceleration=false)
  ]
  rendererClasses [&NiagaraSpriteRendererProperties]
  renderers [{rendererClass=&NiagaraSpriteRendererProperties}]
 }
 ExpFlare {
  emitterUpdateScript [
   &EmitterState()
   &SpawnBurst_Instantaneous(Spawn Count=7, Spawn Time=2.3, Use Spawn Probability=false, UseLoopCountLimit=false)
  ]
  particleSpawnScript [&InitializeParticle(
    Color=<rgba 1 0.9416 0.8226 1>
    Lifetime=0.2
    Position Offset=&AddVectorToPosition(Position=$System.Offset, Vector=&MakeVector(X=0, Y=0, Z=$System.EndHeight))
    Sprite Rotation Mode=::NewEnumerator2
    Sprite Size Max=<300 1000>
    Sprite Size Min=<200 500>
    Sprite Size Mode=::NewEnumerator2
    UsePositionOffset=true
    Write Material Random=true
   ) {module=InitializeParticle001}]
  particleUpdateScript [
   &ParticleState()
   /SetVariables(Particles.Position=&AddVectorToPosition(Position=$Particles.Initial.Position, Vector=&MakeVector(X=0, Y=0, Z=0))) {isSetParameters=true}
   &Color(Color=&ColorFromCurve(ColorCurve=DataInterface() {propertyValues {
       AlphaCurve {keys [<at 0 1 RCIM_Cubic>,<at 1 0 RCIM_Cubic>]}
       BlueCurve {keys [<at 0 0.8674 RCIM_Cubic>,<at 0.1275 0.6971>,<at 0.5125 0.9892 RCIM_Cubic>]}
       GreenCurve {keys [<at 0 0.9669 RCIM_Cubic>,<at 0.1275 0.456>,<at 0.5125 0.0506 RCIM_Cubic>]}
       RedCurve {keys [<at 0 1 RCIM_Cubic>,<at 0.1275 1>,<at 0.5125 1 RCIM_Cubic>]}
      }}))
  ]
  rendererClasses [&NiagaraSpriteRendererProperties]
  renderers [{rendererClass=&NiagaraSpriteRendererProperties}]
 }
 ExpRing {
  simTarget=GPUComputeSim
  emitterUpdateScript [
   &EmitterState()
   &SpawnBurst_Instantaneous(Spawn Time=2, Use Spawn Probability=false, UseLoopCountLimit=false)
  ]
  particleSpawnScript [&InitializeParticle(
    Color=<rgba 2.9887 1.6055 3 1>
    Lifetime=0.7
    Position Offset:NiagaraPosition=$System.Offset
    Position Offset Coordinate Space=::Local
    Sprite Size Mode=::NewEnumerator3
    Uniform Sprite Size=&FloatFromCurve(
     FloatCurve=DataInterface() {propertyValues {Curve {
        defaultValue=1
        keys [
         {arriveTangent=0.1075, interpMode=RCIM_Cubic, leaveTangent=0.1075, tangentMode=RCTM_User, time=0, value=1.075}
         {arriveTangent=0.8933, tangentMode=RCTM_User, time=2.25, value=2}
        ]
       }}}
     CurveIndex=$Emitter.Age
     Scale Curve=88
    )
    UsePositionOffset=true
    Write Material Random=true
   )]
  particleUpdateScript [
   &ParticleState()
   /SetVariables(Particles.Position=&AddVectorToPosition(
     Position=$Particles.Initial.Position
     Vector=&MakeVector(
      X=0
      Y=0
      Z=&FloatFromCurve(
       FloatCurve=DataInterface() {propertyValues {Curve {
          defaultValue=1
          keys [<at 1.5 0 RCIM_Cubic>, {arriveTangent=2.5406, interpMode=RCIM_Cubic, leaveTangent=2.5406, tangentMode=RCTM_User, time=2.25, value=1}]
         }}}
       CurveIndex=$Emitter.Age
       Scale Curve=$System.EndHeight
      )
     )
    )) {isSetParameters=true}
   &ScaleSpriteSize(Uniform Curve Sprite Scale=DataInterface() {propertyValues {Curve {
       defaultValue=0
       keys [
        {interpMode=RCIM_Cubic, tangentMode=RCTM_User, time=0, value=0}
        {arriveTangent=2.8202, interpMode=RCIM_Cubic, leaveTangent=2.8202, tangentMode=RCTM_User, time=0.8, value=1}
       ]
      }}})
   &Color(Color=&ColorFromCurve(ColorCurve=DataInterface() {propertyValues {
       AlphaCurve {keys [<at 0 1 RCIM_Cubic>,<at 1 0 RCIM_Cubic>]}
       BlueCurve {keys [<at 0 0.8674 RCIM_Cubic>,<at 0.1275 1.6054>,<at 0.5125 0.3468 RCIM_Cubic>]}
       GreenCurve {keys [<at 0 0.9669 RCIM_Cubic>,<at 0.1275 0.4957>,<at 0.5125 0.0012 RCIM_Cubic>]}
       RedCurve {keys [<at 0 1 RCIM_Cubic>,<at 0.1275 3>,<at 0.5125 1 RCIM_Cubic>]}
      }}))
  ]
  rendererClasses [&NiagaraSpriteRendererProperties]
  renderers [{rendererClass=&NiagaraSpriteRendererProperties}]
 }
 Flash {
  emitterUpdateScript [
   &EmitterState()
   &SpawnBurst_Instantaneous(Spawn Count=2, Spawn Time=1.35, Use Spawn Probability=false, UseLoopCountLimit=false)
  ]
  particleSpawnScript [&InitializeParticle(
    Color=<rgba 118.6376 3.5204 120 1>
    Lifetime=0.5
    Position Offset:NiagaraPosition=$System.Offset
    Position Offset Coordinate Space=::Local
    Sprite Size Mode=::NewEnumerator3
    Sprite UV Mode=::NewEnumerator1
    Uniform Sprite Size=86.5
    UsePositionOffset=true
    Write Material Random=true
   ) {module=InitializeParticle001}]
  particleUpdateScript [
   &ParticleState()
   &ScaleSpriteSize(
    Initial Sprite Size:ENiagaraCoordinateSpace=$Particles.Initial.SpriteSize
    Uniform Curve Sprite Scale=DataInterface() {propertyValues {Curve {keys [
        {arriveTangent=6.0376, interpMode=RCIM_Cubic, leaveTangent=6.0376, tangentMode=RCTM_User, time=0, value=0}
        {interpMode=RCIM_Cubic, leaveTangent=1.25, time=1, value=1}
       ]}}}
   )
   &CameraOffset(Camera Offset Amount=100) {enabled=false}
   /SetVariables(Particles.Initial.Color:LinearColor=&MultiplyLinearColor(
     Linear Color=$Particles.Initial.Color
     Scale Factor=&MakeVector4_FromFloat(Float=&FloatFromCurve(FloatCurve=DataInterface() {propertyValues {Curve {keys [
           {arriveTangent=1.2591, interpMode=RCIM_Cubic, leaveTangent=1.2591, tangentMode=RCTM_User, time=0, value=0.75}
           {interpMode=RCIM_Cubic, tangentMode=RCTM_User, time=1, value=0}
          ]}}}))
    )) {isSetParameters=true}
   &ScaleColor(
    Scale Alpha=&FloatFromCurve(FloatCurve=DataInterface() {propertyValues {Curve {keys [<at 0 1 RCIM_Cubic>, {arriveTangent=-1.8631, tangentMode=RCTM_User, time=1, value=0}]}}})
    ScaleA=true
    ScaleRGB=true
   )
  ]
  rendererClasses [&NiagaraSpriteRendererProperties]
  renderers [
   {rendererClass=&NiagaraSpriteRendererProperties}
   {rendererClass=&NiagaraSpriteRendererProperties, rendererIndex=1}
  ]
 }
 Flash_Ring {
  emitterUpdateScript [
   &EmitterState()
   &SpawnBurst_Instantaneous(Spawn Time=1.4, Use Spawn Probability=false, UseLoopCountLimit=false)
  ]
  particleSpawnScript [&InitializeParticle(
    Color=<rgba 50 15.8662 26.8898 0.1>
    Lifetime=0.4
    Mass Max=2
    Mass Mode=::NewEnumerator3
    Mesh Scale Mode=::NewEnumerator3
    Mesh Uniform Scale=0.75
    Position Offset:NiagaraPosition=$System.Offset
    Position Offset Coordinate Space=::Local
    UsePositionOffset=true
    Write Material Random=true
   ) {module=InitializeParticle001}]
  particleUpdateScript [
   &ParticleState()
   &ScaleMeshSize(Scale Factor=&VectorFromFloat(Value=&FloatFromCurve(FloatCurve=DataInterface() {propertyValues {Curve {keys [
          {arriveTangent=1.75, interpMode=RCIM_Cubic, leaveTangent=1.75, tangentMode=RCTM_User, time=0, value=0}
          {arriveTangent=0.3461, tangentMode=RCTM_User, time=1, value=1}
         ]}}})))
   &ScaleColor(
    Scale Alpha=&FloatFromCurve(FloatCurve=DataInterface() {propertyValues {Curve {keys [<at 0 1 RCIM_Cubic>,<at 1 0>]}}})
    ScaleA=true
    ScaleRGB=true
   )
  ]
  rendererClasses [&NiagaraMeshRendererProperties]
  renderers [{rendererClass=&NiagaraMeshRendererProperties, rendererIndex=0}]
 }
 Glitter_Rain {
  simTarget=GPUComputeSim
  emitterUpdateScript [
   &EmitterState()
   &SpawnRate(
    SpawnRate=&FloatFromCurve(
     FloatCurve=DataInterface() {propertyValues {Curve {keys [<at 0 0 RCIM_Constant>,<at 1.4 1 RCIM_Constant>,<at 2.4 0 RCIM_Constant>]}}}
     CurveIndex=$Emitter.Age
     Scale Curve=100
    )
    Use Spawn Probability=false
   )
   &SpawnBurst_Instantaneous(Spawn Count=100, Spawn Time=2.4, Use Spawn Probability=false, UseLoopCountLimit=false) {enabled=false}
  ]
  particleSpawnScript [
   &InitializeParticle(
    Color=<rgba 35.9195 2.6087 100 1>
    Lifetime Max=1
    Lifetime Min=0.5
    Lifetime Mode=::NewEnumerator1
    Position Offset=&AddVectorToPosition(Position=$System.Offset, Vector=<0 0 0>)
    Sprite Rotation Mode=::NewEnumerator2
    Sprite Size Mode=::NewEnumerator1
    Uniform Sprite Size Max=3
    Uniform Sprite Size Min=1.5
    UsePositionOffset=true
    Write Material Random=true
   )
   /SetVariables(Particles.Position=&AddVectorToPosition(
     Position=$Particles.Initial.Position
     Vector=&MakeVector(
      X=0
      Y=0
      Z=&FloatFromCurve(
       FloatCurve=DataInterface() {propertyValues {Curve {
          defaultValue=1
          keys [<at 1.5 0 RCIM_Cubic>, {arriveTangent=2.5406, interpMode=RCIM_Cubic, leaveTangent=2.5406, tangentMode=RCTM_User, time=2.25, value=1}]
         }}}
       CurveIndex=$Emitter.Age
       Scale Curve=$System.EndHeight
      )
     )
    )) {isSetParameters=true}
   &ShapeLocation(
    MaskSpawnGroup=false
    Override Seed=false
    Shape Origin=$Particles.Initial.Position
    Sphere Radius=&FloatFromCurve(
     FloatCurve=DataInterface() {propertyValues {Curve {
        defaultValue=1
        keys [
         <at 1.4 1 RCIM_Cubic>
         {arriveTangent=1, interpMode=RCIM_Cubic, leaveTangent=1, time=1.6, value=0.5}
         {arriveTangent=1, interpMode=RCIM_Cubic, leaveTangent=1, time=1.8, value=1.4}
         {arriveTangent=0.875, interpMode=RCIM_Cubic, leaveTangent=0.875, time=2, value=0.9}
         {arriveTangentWeight=0.1795, interpMode=RCIM_Cubic, leaveTangentWeight=0.0898, tangentMode=RCTM_User, time=2.2, value=1.75}
         {arriveTangent=1.25, interpMode=RCIM_Cubic, leaveTangent=1.25, time=2.3, value=1.25}
         <at 2.4 2 RCIM_Cubic>
        ]
       }}}
     CurveIndex=$Emitter.Age
     Scale Curve=20
    )
    UseOwnerScale=true
   )
   &AddVelocity(
    Module.Use Falloff Distance=false
    Origin Offset=&AddVectorToPosition(Position=$System.Offset, Vector=<0 0 72>)
    Override Seed=false
    Velocity Mode=::NewEnumerator1
    Velocity Speed=&RandomRangeFloat(Minimum=150, Maximum=500)
   ) {enabled=false}
   &ApplyInitialForces()
  ]
  particleUpdateScript [
   &ParticleState()
   &ScaleSpriteSize(
    Uniform Curve Scale=&Sine(
     Normalized Angle=$Particles.NormalizedAge
     Period=&RandomRangeFloat(Minimum=0.4, Maximum=0.6)
     Scale=0.5
     Bias=1
    )
    Uniform Curve Sprite Scale=DataInterface() {propertyValues {Curve {defaultValue=0, keys [{tangentMode=RCTM_User, time=0.25, value=1}, {tangentMode=RCTM_User, time=1, value=0}]}}}
   )
   &GravityForce(Gravity=&MakeVector(X=0, Y=0, Z=-245))
   &SolveForcesAndVelocity(Clamp Velocity=false, Limit Acceleration=false)
   &ScaleColor(
    Scale Alpha=&FloatFromCurve(FloatCurve=DataInterface() {propertyValues {Curve {defaultValue=1, keys [{tangentMode=RCTM_User, time=0, value=0}, {tangentMode=RCTM_User, time=0.5, value=1}]}}})
    ScaleA=true
    ScaleRGB=true
   )
  ]
  rendererClasses [&NiagaraSpriteRendererProperties]
  renderers [{rendererClass=&NiagaraSpriteRendererProperties}]
 }
 Glow {
  emitterUpdateScript [
   &EmitterState()
   &SpawnBurst_Instantaneous(Use Spawn Probability=false, UseLoopCountLimit=false)
   &SpawnRate(
    SpawnRate=&FloatFromCurve(
     FloatCurve=DataInterface() {propertyValues {Curve {keys [<at 0 1 RCIM_Constant>,<at 1.4 0 RCIM_Constant>]}}}
     CurveIndex=$Emitter.Age
     Scale Curve=5
    )
    Use Spawn Probability=false
   )
  ]
  particleSpawnScript [&InitializeParticle(
    Color=<rgba 0.5 0.5 0.5 0.5>
    Lifetime=0.5
    Position Offset:NiagaraPosition=$System.Offset
    Sprite Rotation Mode=::NewEnumerator2
    Sprite Size Mode=::NewEnumerator3
    Uniform Sprite Size=75
    UsePositionOffset=true
    Write Material Random=true
   )]
  particleUpdateScript [
   &ParticleState()
   &ScaleSpriteSize(
    Scale Sprite Size Mode=::NewEnumerator0
    Uniform Scale Factor=&FloatFromCurve(FloatCurve=DataInterface() {propertyValues {Curve {keys [<at 0 0>,<at 1 1>]}}})
   ) {enabled=false}
   &ScaleColor(
    Scale Alpha=&FloatFromCurve(FloatCurve=DataInterface() {propertyValues {Curve {
        defaultValue=0
        keys [
         {interpMode=RCIM_Cubic, tangentMode=RCTM_User, time=0, value=0}
         {interpMode=RCIM_Cubic, tangentMode=RCTM_User, time=0.5, value=1}
         {tangentMode=RCTM_User, time=1, value=0}
        ]
       }}})
    ScaleA=true
    ScaleRGB=true
   )
  ]
  rendererClasses [&NiagaraSpriteRendererProperties]
  renderers [{rendererClass=&NiagaraSpriteRendererProperties}]
 }
 Halo {
  simTarget=GPUComputeSim
  emitterUpdateScript [
   &EmitterState()
   &SpawnBurst_Instantaneous(
    Spawn Probability=&FloatFromCurve(
     FloatCurve=DataInterface() {propertyValues {Curve {defaultValue=1, keys [{tangentMode=RCTM_User, time=0, value=0}, {tangentMode=RCTM_User, time=1.4, value=1}]}}}
     CurveIndex=$Emitter.Age
    )
    Spawn Time=1.4
    Use Spawn Probability=true
    UseLoopCountLimit=false
   )
  ]
  particleSpawnScript [&InitializeParticle(
    Color=<rgba 0.3592 0.0261 1 1>
    Position Offset:NiagaraPosition=$System.Offset
    Position Offset Coordinate Space=::Local
    Sprite Size Mode=::NewEnumerator3
    Uniform Sprite Size=42.5
    UsePositionOffset=true
    Write Material Random=true
   )]
  particleUpdateScript [
   &ParticleState()
   /SetVariables(Particles.Position=&AddVectorToPosition(
     Position=$Particles.Initial.Position
     Vector=&MakeVector(
      X=0
      Y=0
      Z=&FloatFromCurve(
       FloatCurve=DataInterface() {propertyValues {Curve {
          defaultValue=1
          keys [<at 1.5 0 RCIM_Cubic>, {arriveTangent=2.5406, interpMode=RCIM_Cubic, leaveTangent=2.5406, tangentMode=RCTM_User, time=2.25, value=1}]
         }}}
       CurveIndex=$Emitter.Age
       Scale Curve=$System.EndHeight
      )
     )
    )) {isSetParameters=true}
   &ScaleSpriteSize(
    Uniform Curve Index=$Emitter.Age
    Uniform Curve Sprite Scale=DataInterface() {propertyValues {Curve {
       defaultValue=1
       keys [
        <at 1.4 1 RCIM_Cubic>
        {
         arriveTangent=1.2445
         arriveTangentWeight=0.1795
         interpMode=RCIM_Cubic
         leaveTangent=1.2445
         leaveTangentWeight=0.0898
         tangentMode=RCTM_User
         time=1.9
         value=1.25
        }
        {arriveTangent=4.1167, arriveTangentWeight=0.05, interpMode=RCIM_Constant, tangentMode=RCTM_User, time=2.3, value=2.875}
       ]
      }}}
   )
   &ScaleColor(
    Scale RGB=&VectorFromFloat(Value=&FloatFromCurve(
      FloatCurve=DataInterface() {propertyValues {Curve {
         defaultValue=1
         keys [
          {tangentMode=RCTM_User, time=0, value=0}
          {arriveTangent=0.38, interpMode=RCIM_Cubic, leaveTangent=0.38, time=1.7, value=0}
          {arriveTangent=1.6522, interpMode=RCIM_Cubic, leaveTangent=1.6522, tangentMode=RCTM_User, time=2.5, value=0.95}
          <at 2.9 0>
         ]
        }}}
      CurveIndex=$Emitter.Age
     ))
    ScaleA=true
    ScaleRGB=true
   )
   &CameraOffset(Camera Offset Amount=100)
  ]
  rendererClasses [&NiagaraSpriteRendererProperties]
  renderers [{rendererClass=&NiagaraSpriteRendererProperties}]
 }
 Heart {
  emitterUpdateScript [
   &EmitterState()
   &SpawnBurst_Instantaneous(Use Spawn Probability=false, UseLoopCountLimit=false)
  ]
  particleSpawnScript [&InitializeParticle(
    Color=<rgba 0.3 0.3 0.3 1>
    Lifetime=1.4
    Mesh Scale Mode=::NewEnumerator3
    Position Offset:NiagaraPosition=$System.Offset
    Sprite Size Mode=::NewEnumerator3
    UsePositionOffset=true
    Write Material Random=true
   )]
  particleUpdateScript [
   &ParticleState()
   /SetVariables(Particles.Position=&AddVectorToPosition(
     Position=$Particles.Initial.Position
     Vector=&MakeVector(
      X=0
      Y=0
      Z=&FloatFromCurve(
       FloatCurve=DataInterface() {propertyValues {Curve {
          defaultValue=1
          keys [<at 1.5 0 RCIM_Cubic>, {arriveTangent=2.5406, interpMode=RCIM_Cubic, leaveTangent=2.5406, tangentMode=RCTM_User, time=2.25, value=1}]
         }}}
       CurveIndex=$Particles.Age
       Scale Curve=$System.EndHeight
      )
     )
    )) {isSetParameters=true}
   /SetVariables(Particles.MeshOrientation=&MakeQuaternion(
     Quaternion From=0
     Axis=<0 0 1>
     Angle=&FloatFromCurve(
      FloatCurve=DataInterface() {propertyValues {Curve {
         defaultValue=0
         keys [
          {arriveTangent=2.5135, leaveTangentWeight=1.4322, time=0, value=0}
          {arriveTangent=0.6667, interpMode=RCIM_Cubic, leaveTangent=0.6667, time=0.5, value=0.25}
          {arriveTangent=0.8643, interpMode=RCIM_Cubic, leaveTangent=0.8643, tangentMode=RCTM_User, time=1.5, value=1}
         ]
        }}}
      CurveIndex=$Particles.Age
      Scale Curve=-1
     )
     From=?
     To=?
     Angle Type=::NewEnumerator1
     Normalize Vectors=false
    )) {isSetParameters=true}
   &ScaleMeshSize(Scale Factor=&VectorFromFloat(Value=&FloatFromCurve(
      FloatCurve=DataInterface() {propertyValues {Curve {
         defaultValue=1
         keys [
          <at 1.4 1 RCIM_Cubic>
          {arriveTangent=1, interpMode=RCIM_Cubic, leaveTangent=1, time=1.6, value=0.5}
          {arriveTangent=1, interpMode=RCIM_Cubic, leaveTangent=1, time=1.8, value=1.4}
          {arriveTangent=0.875, interpMode=RCIM_Cubic, leaveTangent=0.875, time=2, value=0.9}
          {arriveTangentWeight=0.1795, interpMode=RCIM_Cubic, leaveTangentWeight=0.0898, tangentMode=RCTM_User, time=2.2, value=1.75}
          {arriveTangent=11.25, interpMode=RCIM_Cubic, leaveTangent=11.25, time=2.3, value=1.25}
          <at 2.4 4 RCIM_Cubic>
         ]
        }}}
      CurveIndex=$Particles.Age
     )))
   &ScaleMeshSize(Scale Factor=&VectorFromFloat(Value=&FloatFromCurve(
      FloatCurve=DataInterface() {propertyValues {Curve {
         defaultValue=1
         keys [
          <at 1.4 1 RCIM_Cubic>
          {arriveTangent=1, interpMode=RCIM_Cubic, leaveTangent=1, time=1.6, value=0.5}
          {arriveTangent=1, interpMode=RCIM_Cubic, leaveTangent=1, time=1.8, value=1.4}
          {arriveTangent=0.875, interpMode=RCIM_Cubic, leaveTangent=0.875, time=2, value=0.9}
          {arriveTangentWeight=0.1795, interpMode=RCIM_Cubic, leaveTangentWeight=0.0898, tangentMode=RCTM_User, time=2.2, value=1.75}
          {arriveTangent=11.25, interpMode=RCIM_Cubic, leaveTangent=11.25, time=2.3, value=1.25}
          <at 2.4 4 RCIM_Cubic>
         ]
        }}}
      CurveIndex=$Particles.Age
     ))) {enabled=false, module=ScaleMeshSize001}
   &SolveForcesAndVelocity(Clamp Velocity=false, Limit Acceleration=false)
  ]
  rendererClasses [&NiagaraSpriteRendererProperties, &NiagaraMeshRendererProperties]
  renderers [
   {rendererClass=&NiagaraSpriteRendererProperties}
   {rendererClass=&NiagaraMeshRendererProperties, rendererIndex=1}
  ]
 }
 InnerHeart {
  emitterUpdateScript [
   &EmitterState()
   &SpawnBurst_Instantaneous(Spawn Count=2, Use Spawn Probability=false, UseLoopCountLimit=false)
  ]
  particleSpawnScript [&InitializeParticle(
    Color=<rgba 7 0 6.7977 1>
    Lifetime=2.4
    Mesh Scale Mode=::NewEnumerator3
    Position Offset:NiagaraPosition=$System.Offset
    Sprite Size Mode=::NewEnumerator3
    Uniform Sprite Size=150
    UsePositionOffset=true
    Write Material Random=true
   )]
  particleUpdateScript [
   &ParticleState()
   &ScaleSpriteSize(
    Uniform Curve Index=$Particles.Age
    Uniform Curve Scale=0.272
    Uniform Curve Sprite Scale=DataInterface() {propertyValues {Curve {
       defaultValue=1
       keys [
        {arriveTangent=0.7793, interpMode=RCIM_Cubic, leaveTangent=0.7793, tangentMode=RCTM_User, time=0, value=0}
        {arriveTangent=0.642, interpMode=RCIM_Cubic, leaveTangent=0.642, tangentMode=RCTM_Break, time=1, value=1}
        {arriveTangent=5.6292, leaveTangent=16.6239, tangentMode=RCTM_User, time=2.4, value=3.75}
       ]
      }}}
   )
   /SetVariables(Particles.Position=&AddVectorToPosition(
     Position=$Particles.Initial.Position
     Vector=&MakeVector(
      X=0
      Y=0
      Z=&FloatFromCurve(
       FloatCurve=DataInterface() {propertyValues {Curve {
          defaultValue=1
          keys [<at 1.5 0 RCIM_Cubic>, {arriveTangent=2.5406, interpMode=RCIM_Cubic, leaveTangent=2.5406, tangentMode=RCTM_User, time=2.25, value=1}]
         }}}
       CurveIndex=$Particles.Age
       Scale Curve=$System.EndHeight
      )
     )
    )) {isSetParameters=true}
   /SetVariables(Particles.MeshOrientation=&MakeQuaternion(
     Quaternion From=0
     Axis=<0 0 1>
     Angle=&FloatFromCurve(
      FloatCurve=DataInterface() {propertyValues {Curve {
         defaultValue=0
         keys [
          {arriveTangent=2.5135, leaveTangentWeight=1.4322, time=0, value=0}
          {arriveTangent=0.6667, interpMode=RCIM_Cubic, leaveTangent=0.6667, time=0.5, value=0.25}
          {arriveTangent=0.8643, interpMode=RCIM_Cubic, leaveTangent=0.8643, tangentMode=RCTM_User, time=1.5, value=1}
         ]
        }}}
      CurveIndex=$Particles.Age
      Scale Curve=-1
     )
     From=?
     To=?
     Angle Type=::NewEnumerator1
     Normalize Vectors=false
    )) {isSetParameters=true}
   &ScaleColor(
    Scale RGB=&VectorFromFloat(Value=&FloatFromCurve(FloatCurve=DataInterface() {propertyValues {Curve {keys [
          <at 0 1>
          {arriveTangent=-29.3067, leaveTangent=-29.3067, time=0, value=1}
          {arriveTangent=5.6088, interpMode=RCIM_Cubic, leaveTangent=5.6088, tangentMode=RCTM_User, time=0.8, value=1}
          {arriveTangent=291.0163, interpMode=RCIM_Cubic, leaveTangent=291.0163, tangentMode=RCTM_User, time=1, value=30}
         ]}}}))
    ScaleA=true
    ScaleRGB=true
   ) {enabled=false}
   &ScaleColor(
    Scale Alpha=0.264
    Scale RGB=&VectorFromFloat(Value=&FloatFromCurve(
      FloatCurve=DataInterface() {propertyValues {Curve {keys [
          {arriveTangent=-29.3067, time=0, value=0.2}
          <at 0 0.2>
          {arriveTangent=0.5714, interpMode=RCIM_Cubic, leaveTangent=0.5714, time=1.3024, value=0.2}
          {arriveTangent=0.3781, interpMode=RCIM_Cubic, leaveTangent=0.3781, tangentMode=RCTM_User, time=1.4, value=1}
          {interpMode=RCIM_Cubic, leaveTangent=169.9066, time=1.6996, value=1}
         ]}}}
      CurveIndex=$Particles.Age
     ))
    ScaleA=true
    ScaleRGB=true
   ) {module=ScaleColor001}
   &ScaleMeshSize(Scale Factor=&VectorFromFloat(Value=&FloatFromCurve(
      FloatCurve=DataInterface() {propertyValues {Curve {
         defaultValue=1
         keys [
          <at 1.4 1 RCIM_Cubic>
          {arriveTangent=1, interpMode=RCIM_Cubic, leaveTangent=1, time=1.6, value=0.5}
          {arriveTangent=1, interpMode=RCIM_Cubic, leaveTangent=1, time=1.8, value=1.4}
          {arriveTangent=0.875, interpMode=RCIM_Cubic, leaveTangent=0.875, time=2, value=0.9}
          {arriveTangentWeight=0.1795, interpMode=RCIM_Cubic, leaveTangentWeight=0.0898, tangentMode=RCTM_User, time=2.2, value=1.75}
          {arriveTangent=11.25, interpMode=RCIM_Cubic, leaveTangent=11.25, time=2.3, value=1.25}
          <at 2.4 4 RCIM_Cubic>
         ]
        }}}
      CurveIndex=$Particles.Age
      Scale Curve=&FloatFromCurve(
       FloatCurve=DataInterface() {propertyValues {Curve {defaultValue=1, keys [{tangentMode=RCTM_User, time=0, value=0}, <at 1.3 0.4>, {tangentMode=RCTM_User, time=1.4, value=1}]}}}
       CurveIndex=$Particles.Age
      )
     ))) {enabled=false}
   &ScaleMeshSize(Scale Factor=&VectorFromFloat(Value=&FloatFromCurve(
      FloatCurve=DataInterface() {propertyValues {Curve {
         defaultValue=1
         keys [
          {arriveTangent=0.7793, interpMode=RCIM_Cubic, leaveTangent=0.7793, tangentMode=RCTM_User, time=0, value=0}
          {arriveTangent=0.642, interpMode=RCIM_Cubic, leaveTangent=0.642, tangentMode=RCTM_Break, time=1, value=1}
          {arriveTangent=5.6292, leaveTangent=16.6239, tangentMode=RCTM_User, time=2.4, value=3.75}
         ]
        }}}
      CurveIndex=$Particles.Age
     ))) {module=ScaleMeshSize001}
  ]
  rendererClasses [&NiagaraSpriteRendererProperties, &NiagaraMeshRendererProperties]
  renderers [
   {rendererClass=&NiagaraSpriteRendererProperties}
   {rendererClass=&NiagaraMeshRendererProperties, rendererIndex=1}
  ]
 }
 Orbs_Ribbon {
  emitterUpdateScript [&EmitterState()]
  particleSpawnScript [&InitializeParticle(
    Color Mode=::NewEnumerator0
    Lifetime=0.5
    Ribbon Width=3
    Ribbon Width Mode=::NewEnumerator1
    UsePositionOffset=false
    Write Material Random=true
   )]
  particleUpdateScript [
   &ParticleState()
   &SolveForcesAndVelocity(Clamp Velocity=false, Limit Acceleration=false)
   &ScaleRibbonWidth(Ribbon Width Scale=&FloatFromCurve(FloatCurve=DataInterface() {propertyValues {Curve {keys [<at 0 1>,<at 1 0>]}}}))
   &KillParticles(Kill Particles=&SetBoolByFloatComparison(A=1.4, B=$Emitter.Age, Comparison Type=::NewEnumerator6))
  ]
  rendererClasses [&NiagaraRibbonRendererProperties]
  renderers [{rendererClass=&NiagaraRibbonRendererProperties}]
 }
 Swirly_Orbs {
  emitterUpdateScript [
   &EmitterState()
   &SpawnBurst_Instantaneous(Spawn Count=5, Use Spawn Probability=false, UseLoopCountLimit=false)
   &SpawnRate(
    SpawnRate=&Multiply_Float(
     A=8
     B=&FloatFromCurve(FloatCurve=DataInterface() {propertyValues {Curve {keys [<at 0 0>,<at 0.1 1>,<at 0.9 1>,<at 1 0>]}}}, CurveIndex=$Emitter.Age)
    )
    Use Spawn Probability=false
   )
  ]
  particleSpawnScript [
   &InitializeParticle(
    Color=<rgba 3.5919 0.2609 10 1>
    Lifetime=&FloatFromCurve(FloatCurve=DataInterface() {propertyValues {Curve {keys [<at 0 1>,<at 1.4 0>]}}}, CurveIndex=$Emitter.Age, Scale Curve=1.3)
    Position Offset:NiagaraPosition=$System.Offset
    Ribbon Width Mode=::NewEnumerator1
    Sprite Size Mode=::NewEnumerator1
    Uniform Sprite Size Max=8
    Uniform Sprite Size Min=4
    UsePositionOffset=true
    Write Material Random=true
   )
   /SetVariables(Particles.SpherePos:NiagaraPosition=$Particles.Position) {isSetParameters=true}
   &ShapeLocation(
    MaskSpawnGroup=false
    Override Seed=false
    Shape Origin=$Particles.SpherePos
    Sphere Radius=75
    Sphere Surface Distribution=0.75
    UseOwnerScale=true
   ) {enabled=false}
   &AddVelocity(
    Module.Use Falloff Distance=false
    Origin Offset:NiagaraPosition=$System.Offset
    Override Seed=false
    Velocity Mode=::NewEnumerator1
    Velocity Speed=-250
   ) {enabled=false}
  ]
  particleUpdateScript [
   &ParticleState()
   &RotateAroundPoint(
    Derive the Rotation Axis=false
    Radius:NiagaraFloat=&FloatFromCurve(FloatCurve=DataInterface() {propertyValues {Curve {keys [<at 0 1>,<at 1 0>]}}}, Scale Curve=37.5)
    Rotation Axis:Vector3f=&RandomRangeVector(Minimum=<-0.5 -0.5 0.1>, Maximum=<0.5 0.5 1>)
    Rotation Center:NiagaraPosition=$Particles.SpherePos
    Rotation Phase:NiagaraFloat=&RandomRangeFloat(Minimum=0, Maximum=1)
    Rotation Rate:NiagaraFloat=&RandomRangeFloat(Minimum=-0.25, Maximum=-0.75)
   )
   /SetVariables(Particles.Position=&AddVectorToPosition(
     Position=$Output.RotateAroundPoint.OutputPosition
     Vector=&MakeVector(
      X=0
      Y=&FloatFromCurve(
       FloatCurve=DataInterface() {propertyValues {Curve {keys [<at 0 1>,<at 1 0>]}}}
       Scale Curve=&Sine(Normalized Angle=$Emitter.Age, Period=&RandomRangeFloat(Minimum=0.4, Maximum=0.6), Scale=2.5, Bias=0)
      )
      Z=&FloatFromCurve(
       FloatCurve=DataInterface() {propertyValues {Curve {keys [<at 0 1>,<at 1 0>]}}}
       Scale Curve=&Sine(Normalized Angle=$Emitter.Age, Period=&RandomRangeFloat(Minimum=0.4, Maximum=0.6), Scale=2.5, Bias=0)
      )
     )
    )) {enabled=false, isSetParameters=true}
   &SolveForcesAndVelocity(Clamp Velocity=false, Limit Acceleration=false)
   &GenerateLocationEvent(Event Send Rate=60, Use Event Probability=false, UseEventDelay=false)
   &ScaleSpriteSizeBySpeed(Max Scale Factor=<1 75>) {enabled=false}
   &KillParticlesInVolume(Sphere Radius=4, Volume Origin:NiagaraPosition=$Particles.SpherePos) {enabled=false}
  ]
  rendererClasses [&NiagaraSpriteRendererProperties]
  renderers [{rendererClass=&NiagaraSpriteRendererProperties}]
 }
}
summary {
 systemName=NS_AdHoc_Squad_Burst_01
 emitters [
  {emitterName=Heart, rendererClasses [&NiagaraSpriteRendererProperties, &NiagaraMeshRendererProperties]}
  {emitterName=InnerHeart, rendererClasses [&NiagaraSpriteRendererProperties, &NiagaraMeshRendererProperties]}
  {emitterName=Halo, simTarget=GPUComputeSim, rendererClasses [&NiagaraSpriteRendererProperties]}
  {emitterName=Burst_Sparks, simTarget=GPUComputeSim, rendererClasses [&NiagaraSpriteRendererProperties]}
  {emitterName=Glitter_Rain, simTarget=GPUComputeSim, rendererClasses [&NiagaraSpriteRendererProperties]}
  {emitterName=Burst_Twinkles, simTarget=GPUComputeSim, rendererClasses [&NiagaraSpriteRendererProperties]}
  {emitterName=Swirly_Orbs, rendererClasses [&NiagaraSpriteRendererProperties]}
  {emitterName=Orbs_Ribbon, rendererClasses [&NiagaraRibbonRendererProperties]}
  {emitterName=Flash, rendererClasses [&NiagaraSpriteRendererProperties]}
  {emitterName=Flash_Ring, rendererClasses [&NiagaraMeshRendererProperties]}
  {emitterName=ExpRing, simTarget=GPUComputeSim, rendererClasses [&NiagaraSpriteRendererProperties]}
  {emitterName=ExpFlare, rendererClasses [&NiagaraSpriteRendererProperties]}
  {emitterName=Burst_Sparks_B, simTarget=GPUComputeSim, rendererClasses [&NiagaraSpriteRendererProperties]}
  {emitterName=Glow, rendererClasses [&NiagaraSpriteRendererProperties]}
 ]
 userVariables [User.CoreOffset {
   type=ENiagaraOrientationAxis
   description=""
   defaultValue {
    struct=NiagaraExt_VariableValue_Enum
    value {displayName=X Axis, enum=/Script/Niagara.ENiagaraOrientationAxis, enumName=ENiagaraOrientationAxis::XAxis}
   }
  }]
}
