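# Looping Niagara cosmetic effect for AdHoc Squad, showing a rotating "Heart" mesh with a pulsing glow, both placed at a system-level offset.
# A GPU Heart emitter bursts mesh particles that last about one loop, slightly padding some lifetimes by loop count, likely to hide seams.
# System age drives the initial scale-in, while normalized age drives Z rotation.
# A CPU Glow emitter continuously spawns short-lived sprites that fade in and out.
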NiagaraSystem /BRCosmetics/Athena/Effects/Particles/AdHocSquad/NS_AdHoc_Squad_Loop_01
dependencies {
 usedDataInterfaces [&NiagaraDataInterfaceCurve]
 usedDynamicInputs []
 usedModules [
  &self:SystemSpawnScript.SystemScriptSource.SystemScriptGraph.NiagaraNodeAssignment_0.SetVariables_#1
  &SystemState
 ]
}
emitters {
 Glow {
  emitterUpdateScript [
   &EmitterState()
   &SpawnBurst_Instantaneous(Use Spawn Probability=false, UseLoopCountLimit=true)
   &SpawnRate(SpawnRate=5, Use Spawn Probability=false)
  ]
  particleSpawnScript [&InitializeParticle(
    Color=<rgba 0.5 0.5 0.5 0.5>
    Lifetime=0.5
    Position Offset:NiagaraPosition=$System.Offset
    Sprite Rotation Mode=::NewEnumerator2
    Sprite Size Mode=::NewEnumerator3
    Uniform Sprite Size=75
    UsePositionOffset=true
    Write Material Random=true
   )]
  particleUpdateScript [
   &ParticleState()
   &ScaleSpriteSize(
    Scale Sprite Size Mode=::NewEnumerator0
    Uniform Scale Factor=&FloatFromCurve(FloatCurve=DataInterface() {propertyValues {Curve {keys [<at 0 0>,<at 1 1>]}}})
   ) {enabled=false}
   &ScaleColor(
    Scale Alpha=&FloatFromCurve(FloatCurve=DataInterface() {propertyValues {Curve {
        defaultValue=0
        keys [
         {interpMode=RCIM_Cubic, tangentMode=RCTM_User, time=0, value=0}
         {interpMode=RCIM_Cubic, tangentMode=RCTM_User, time=0.5, value=1}
         {tangentMode=RCTM_User, time=1, value=0}
        ]
       }}})
    ScaleA=true
    ScaleRGB=true
   )
  ]
  rendererClasses [&NiagaraSpriteRendererProperties]
  renderers [{rendererClass=&NiagaraSpriteRendererProperties}]
 }
 Heart {
  simTarget=GPUComputeSim
  emitterUpdateScript [
   &EmitterState()
   &SpawnBurst_Instantaneous(Use Spawn Probability=false, UseLoopCountLimit=false)
  ]
  particleSpawnScript [&InitializeParticle(
    Color=<rgba 0.3 0.3 0.3 1>
    Lifetime=&MakeCustomFloatFromBool(
     Bool=&SetBoolByFloatComparison(A=&MakeFloatFromInt(Int=$System.LoopCount), B=0)
     True Float=$System.CurrentLoopDuration
     False Float=&Add_Float(A=$System.CurrentLoopDuration, B=0.05)
    )
    Mesh Scale Mode=::NewEnumerator3
    Position Offset:NiagaraPosition=$System.Offset
    Sprite Size Mode=::NewEnumerator3
    Uniform Sprite Size=100
    UsePositionOffset=true
    Write Material Random=true
   )]
  particleUpdateScript [
   &SolveForcesAndVelocity(Clamp Velocity=false, Limit Acceleration=false)
   &ParticleState()
   &ScaleMeshSize(Scale Factor=&VectorFromFloat(Value=&FloatFromCurve(
      FloatCurve=DataInterface() {propertyValues {Curve {defaultValue=1, keys [{tangentMode=RCTM_User, time=0, value=0}, {tangentMode=RCTM_User, time=0.15, value=1}]}}}
      CurveIndex=$Engine.System.Age
     )))
   /SetVariables(Particles.MeshOrientation=&MakeQuaternion(
     Quaternion From=0
     Axis=<0 0 1>
     Angle=&Multiply_Float(A=$Particles.NormalizedAge, B=-0.5)
     From=?
     To=?
     Angle Type=::NewEnumerator1
     Normalize Vectors=false
    )) {isSetParameters=true}
  ]
  rendererClasses [&NiagaraMeshRendererProperties]
  renderers [{rendererClass=&NiagaraMeshRendererProperties, rendererIndex=0}]
 }
}
summary {
 systemName=NS_AdHoc_Squad_Loop_01
 emitters [
  {emitterName=Heart, simTarget=GPUComputeSim, rendererClasses [&NiagaraMeshRendererProperties]}
  {emitterName=Glow, rendererClasses [&NiagaraSpriteRendererProperties]}
 ]
 userVariables []
}
